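# Physics asset for the TurnVortex backpack cosmetic, defining ragdoll-style bodies and joints for four movable pieces, top and bottom in a/b pairs.
# Each piece pairs a tiny capsule on a base bone with a longer, offset capsule on its dyn bone.
# Four independent constraints join each dyn bone to its base, with a moderate swing limit and almost no twist range.
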PhysicsAsset /BRCosmetics/Accessories/FORT_Backpacks/Backpack_F_MED_TurnVortex/Meshes/F_MED_TurnVortex_Pack_Physics
bodies {
 base_bottom_a [{shapeType=Capsule, length=0.505, radius=0.505, shapeName=base_bottom_a_capsule, rotation=<rot 0 -90 0>}]
 base_bottom_b [{shapeType=Capsule, length=0.505, radius=0.505, shapeName=base_bottom_b_capsule, rotation=<rot 0 -90 0>}]
 base_top_a [{shapeType=Capsule, length=0.505, radius=0.505, shapeName=base_top_a_capsule, rotation=<rot 0 -90 0>}]
 base_top_b [{shapeType=Capsule, length=0.505, radius=0.505, shapeName=base_top_b_capsule, rotation=<rot 0 -90 0>}]
 dyn_bottom_a_00 [{shapeType=Capsule, length=4, radius=1, shapeName=dyn_bottom_a_00_capsule, center=<2 0 0>, rotation=<rot 90 0 0>}]
 dyn_bottom_b_00 [{shapeType=Capsule, length=4, radius=1, shapeName=dyn_bottom_b_00_capsule, center=<2 0 0>, rotation=<rot 90 0 0>}]
 dyn_top_a_00 [{shapeType=Capsule, length=4, radius=1, shapeName=dyn_top_a_00_capsule, center=<2 0 0>, rotation=<rot 90 0 0>}]
 dyn_top_b_00 [{shapeType=Capsule, length=4, radius=1, shapeName=dyn_top_b_00_capsule, center=<2 0 0>, rotation=<rot 90 0 0>}]
}
constraints [
 {bone1Name=dyn_bottom_a_00, bone2Name=base_bottom_a, swing1LimitDegrees=45, swing2LimitDegrees=45, twistLimitDegrees=1}
 {bone1Name=dyn_bottom_b_00, bone2Name=base_bottom_b, swing1LimitDegrees=45, swing2LimitDegrees=45, twistLimitDegrees=1}
 {bone1Name=dyn_top_a_00, bone2Name=base_top_a, swing1LimitDegrees=45, swing2LimitDegrees=45, twistLimitDegrees=1}
 {bone1Name=dyn_top_b_00, bone2Name=base_top_b, swing1LimitDegrees=45, swing2LimitDegrees=45, twistLimitDegrees=1}
]
