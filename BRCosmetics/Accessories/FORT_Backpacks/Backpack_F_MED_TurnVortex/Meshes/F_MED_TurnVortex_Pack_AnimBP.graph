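# Animation Blueprint for the TurnVortex backpack cosmetic that adds simulated physics motion to the pack's skeleton.
# A RigidBody node simulates the incoming pose in root-bone space using the pack physics asset, then an additive bone-space offset nudges the pack bone.
# Each update sets Gravity Override downward while skydiving and upward otherwise.
# The RigidBody node overrides world gravity, but its gravity pin shows no link to that variable.
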AnimBlueprint /BRCosmetics/Accessories/FORT_Backpacks/Backpack_F_MED_TurnVortex/Meshes/F_MED_TurnVortex_Pack_AnimBP
target_skeleton=&/F_MED_TurnVortex_Pack_Skeleton
graphs [
 AnimGraph {nodes [
   n1 = RigidBody!(ComponentPose=Local To Component(LocalPose=Input Pose()), Alpha, OverrideWorldGravity:Vector="0,0,0", ExternalForce) -> (Pose) {
    bClampLinearTranslationLimitToRefPose=true
    bOverrideWorldGravity=true
    overridePhysicsAsset=&/F_MED_TurnVortex_Pack_Physics
    simulationSpace=BaseBoneSpace
    baseBoneRef {boneName=root}
    componentAppliedLinearAccClamp=<0 0 0>
    simSpaceSettings {
     maxAngularAcceleration=1000
     maxAngularVelocity=3
     maxLinearAcceleration=300
     maxLinearVelocity=700
     worldAlpha=1
     externalLinearDragV {y=0.5, z=0.5}
    }
   }
   Output Pose(Result=Component To Local(ComponentPose="Transform (Modify) Bone - Bone: pack"(ComponentPose=@n1, Translation="0,2,0", Rotation="0,0,0") {translationMode=BMM_Additive, translationSpace=BCS_BoneSpace, boneToModify {boneName=pack}, translation {y=2}})) {layerGroup=None}
  ]}
 EventGraph {nodes [
   Event BlueprintUpdateAnimation()
   > n1 = Branch(Condition=Get bIsSkyDiving())
   > Set Gravity Override(Gravity Override="0,0,-980")
   Set Gravity Override(execute=@n1.else, Gravity Override="0,0,980")
   TryGetPawnOwner()
  ]}
]
variables [Gravity Override {type=Vector}]
